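# Translucent unlit material for Niagara ribbons, a temporary idle fire effect for the SolidMist pickaxe.
# Two panning fire trail samples multiply into a flame, colored through a golden-fire curve atlas and masked by two opposed fire-tip textures with depth fade.
# Gaussian noise normals pan with opposite vertical motion on each V half, forming UV distortion reroutes.
# A camera offset function is also present.
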Material /BRCosmetics/Weapons/FORT_Melee/Pickaxe_SolidMist/FX/Temp/M_SolidMist_Pickaxe_Idle
graph {
 properties {
  bAutomaticallySetUsageInEditor=false
  bUsedWithNiagaraRibbons=true
  blendMode=BLEND_Translucent
  shadingModel=MSM_Unlit
  usedShadingModels=Unlit
 }
 nodes [
  CurveAtlasRowParameter() {
   atlas=/Game/Effects/Fort_Effects/GameplayPulginsAssets/Demon/CA_GoldenFire
   curve=/Game/Effects/Fort_Effects/GameplayPulginsAssets/Demon/Curve_GoldenFire02
   defaultValue=1
   parameterName=Param_1
  }
  n1 = Add(A=NamedRerouteUsage(), B=TextureCoordinate())
  n2 = ScalarParameter() {defaultValue=3, parameterName=Param_1}
  n3 = Constant() {r=1}
  n4 = MaterialFunctionCall(UV Channel=Multiply(A=@n1, B=AppendVector(A=@n2, B=@n3))) {materialFunction=&LinearGradient}
  n5 = AppendVector(A=OneMinus(None=@n4.UGradient), B=@n4.VGradient)
  n6 = MaterialFunctionCall(UV Channel=MaterialFunctionCall(UVs=Multiply(A=@n1), Texture Scale=Constant2Vector() {g=1.5, r=0.7}) {materialFunction=&ScaleUVsByCenter}.UVs) {materialFunction=&LinearGradient}
  n7 = AppendVector(A=OneMinus(None=@n6.UGradient), B=@n6.VGradient)
  n8 = Multiply(
   A=TextureSample(UVs=Panner(Coordinate=@n5) {speedX=-0.25}) {texture=/Game/Effects/Fort_Effects/Textures/Fire/T_Stylized_Fire_Trail}.R
   B=TextureSample(UVs=Panner(Coordinate=@n7) {speedX=-0.35}) {texture=&T_Stylized_Fire_Trail}.R
  )
  n9 = ScalarParameter() {defaultValue=1, parameterName=Param_16}
  n10 = Divide(A=NamedRerouteUsage(), B=ScalarParameter() {defaultValue=13, parameterName=Param_13})
  n11 = MaterialFunctionCall(UV Channel=TextureCoordinate()) {materialFunction=&LinearGradient}
  n12 = AppendVector(A=Add(A=@n11.UGradient, B=ScalarParameter() {defaultValue=0.4, parameterName=Param_4}), B=@n11.VGradient)
  n13 = AppendVector(A=ScalarParameter() {defaultValue=0.2, parameterName=Param_2}, B=ScalarParameter() {defaultValue=1, parameterName=Param_3})
  n14 = TextureSample(UVs=Add(
    A=@n10
    B=MaterialFunctionCall(UVs=@n12, Texture Scale=@n13) {materialFunction=&ScaleUVsByCenter}.UVs
   )) {samplerSource=SSM_Clamp_WorldGroupSettings, texture=&/T_SolidMist_Fire_Tip}
  n15 = MaterialFunctionCall(UV Channel=TextureCoordinate()) {materialFunction=&LinearGradient}
  n16 = ScalarParameter() {defaultValue=-0.45, parameterName=Param_14}
  n17 = MaterialFunctionCall(
   UVs=AppendVector(A=Add(A=@n15.UGradient, B=@n16), B=@n15.VGradient)
   Texture Scale=AppendVector(A=ScalarParameter() {defaultValue=0.1, parameterName=Param_15}, B=ScalarParameter() {defaultValue=1, parameterName=Param_3})
  ) {materialFunction=&ScaleUVsByCenter}
  n18 = TextureSample(UVs=Add(A=@n10, B=OneMinus(None=@n17.UVs))) {samplerSource=SSM_Clamp_WorldGroupSettings, texture=&T_SolidMist_Fire_Tip}
  DepthFade(Opacity=Multiply(A=Power(Base=@n8, Exp=@n9), B=Multiply(A=@n14.R, B=@n18.R))) {fadeDistanceDefault=20}
  n19 = CurveAtlasRowParameter(CurveTime=@n8) {
   atlas=&CA_GoldenFire
   curve=/Game/Effects/Fort_Effects/GameplayPulginsAssets/Demon/Curve_GoldenFire01
   parameterName=Param_2
  }
  MaterialFunctionCall(Emissive=Multiply(A=Multiply(A=@n19), B=ScalarParameter() {defaultValue=2})) {materialFunction=&MF_FX_EmissiveAdjustments}
  MaterialFunctionCall() {materialFunction=&LinearGradient}
  MaterialFunctionCall(Offset Amount=ScalarParameter() {parameterName=Param_10}) {materialFunction=/Engine/Functions/Engine_MaterialFunctions02/WorldPositionOffset/CameraOffset}
  n20 = ScalarParameter() {defaultValue=3, parameterName=Param_6}
  n21 = ScalarParameter() {defaultValue=1.6, parameterName=Param_7}
  n22 = Multiply(A=TextureCoordinate(), B=Divide(A=AppendVector(A=@n20, B=@n21), B=ScalarParameter() {defaultValue=1, parameterName=Param_11}))
  n23 = ScalarParameter() {defaultValue=0.5, parameterName=Param_8}
  n24 = ScalarParameter() {defaultValue=3, parameterName=Param_9}
  n25 = AppendVector(A=@n23, B=@n24)
  n26 = ScalarParameter() {defaultValue=3, parameterName=Param_12}
  n27 = TextureSample(UVs=Panner(Coordinate=@n22, Speed=Divide(A=@n25, B=@n26)) {speedX=0.2, speedY=0.2}) {samplerType=SAMPLERTYPE_Normal, texture=/Game/Characters/Player/Female/Medium/Bodies/F_MED_Bistro_Astronaut/FX/T_GaussianNoise_Normal}
  n28 = Round(None=MaterialFunctionCall() {materialFunction=&LinearGradient}.VGradient)
  n29 = Multiply(A=ComponentMask(None=@n27.RGB) {g=true, r=true}, B=OneMinus(None=@n28))
  n30 = AppendVector(A=@n23, B=Multiply(A=@n24) {constB=-1})
  n31 = TextureSample(UVs=Panner(Coordinate=@n22, Speed=Divide(A=@n30, B=@n26)) {speedX=0.2, speedY=-0.2}) {samplerType=SAMPLERTYPE_Normal, texture=&T_GaussianNoise_Normal}
  n32 = ComponentMask(None=@n31.RGB) {g=true, r=true}
  n33 = Multiply(A=@n32, B=@n28)
  NamedRerouteDeclaration(None=Divide(A=Add(A=@n29, B=@n33), B=ScalarParameter() {defaultValue=30, parameterName=Param_5})) {name=UVDistortionOpposite, nodeColor=<rgba 0 0.7656 1 1>}
  NamedRerouteDeclaration(None=@n32) {name=UVDistortion, nodeColor=<rgba 0.5938 0 1 1>}
  Panner()
  TextureSample() {samplerType=SAMPLERTYPE_Normal, texture=/Game/Effects/Fort_Effects/Textures/Fire/T_WispyFireNoise_Soft_N}
  TextureSample() {samplerType=SAMPLERTYPE_Grayscale, texture=/Game/Effects/Fort_Effects/Textures/Fire/T_WispyFireNoise_Soft_M}
  VectorParameter() {defaultValue=<rgba 1 0.4246 0.075 1>}
 ]
}
