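# M_Temp is a minimal Default Lit material in a Temp folder under the Solid Mist pickaxe's FX content, apparently a placeholder.
# Its graph holds only a constant and a dark blue vector parameter, with no textures, math or visible connections.
# Usage is enabled for skeletal meshes and mesh deformers, and automatic usage setting in the editor is off.
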Material /BRCosmetics/Weapons/FORT_Melee/Pickaxe_SolidMist/FX/Temp/M_Temp
graph {
 properties {bAutomaticallySetUsageInEditor=false, bUsedWithMeshDeformer=true, bUsedWithSkeletalMesh=true, usedShadingModels=Default Lit}
 nodes [
  Constant()
  VectorParameter() {defaultValue=<rgba 0.0111 0.0238 0.0451 1>}
 ]
}
