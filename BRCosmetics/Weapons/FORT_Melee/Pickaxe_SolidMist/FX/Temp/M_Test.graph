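# M_Test is an unlit, translucent, two-sided fire-trail material for Niagara ribbons, in the SolidMist pickaxe's Temp FX folder.
# Two noise normals pan in opposite vertical directions, each on one half of V, to distort the panning fire-trail texture.
# Its red channel indexes a golden-fire curve atlas for emissive color, tinted by particle color.
# It also drives depth-faded opacity, with dynamic parameters controlling end fades, erosion and U offset.
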Material /BRCosmetics/Weapons/FORT_Melee/Pickaxe_SolidMist/FX/Temp/M_Test
graph {
 properties {
  bAutomaticallySetUsageInEditor=false
  bUsedWithNiagaraRibbons=true
  blendMode=BLEND_Translucent
  shadingModel=MSM_Unlit
  twoSided=true
  usedShadingModels=Unlit
 }
 nodes [
  n1 = Constant() {r=3}
  n2 = Constant() {r=1.6}
  n3 = Multiply(A=TextureCoordinate(), B=Divide(A=AppendVector(A=@n1, B=@n2), B=Constant() {r=1}))
  n4 = Constant() {r=0.5}
  n5 = Constant() {r=3}
  n6 = AppendVector(A=@n4, B=@n5)
  n7 = Constant() {r=3}
  n8 = TextureSample(UVs=Panner(Coordinate=@n3, Speed=Divide(A=@n6, B=@n7)) {speedX=0.2, speedY=0.2}) {samplerType=SAMPLERTYPE_Normal, texture=/Game/Characters/Player/Female/Medium/Bodies/F_MED_Bistro_Astronaut/FX/T_GaussianNoise_Normal}
  n9 = ComponentMask(None=@n8.RGB) {g=true, r=true}
  n10 = Round(None=MaterialFunctionCall() {materialFunction=&LinearGradient}.VGradient)
  n11 = OneMinus(None=@n10)
  n12 = AppendVector(A=@n4, B=Multiply(A=@n5) {constB=-1})
  n13 = TextureSample(UVs=Panner(Coordinate=@n3, Speed=Divide(A=@n12, B=@n7)) {speedX=0.2, speedY=-0.2}) {samplerType=SAMPLERTYPE_Normal, texture=&T_GaussianNoise_Normal}
  n14 = ComponentMask(None=@n13.RGB) {g=true, r=true}
  n15 = Multiply(
   A=Add(A=Multiply(A=@n9, B=@n11), B=Multiply(A=@n14, B=@n10))
   B=LinearInterpolate(Alpha=ScalarParameter() {parameterName=UV_Distortion}) {constB=0.1}
  )
  n16 = TextureCoordinate()
  n17 = Constant() {r=1}
  n18 = Constant() {r=1.2}
  n19 = MaterialFunctionCall(Float2=MaterialFunctionCall(UVs=Add(A=@n15, B=@n16), Texture Scale=AppendVector(A=@n17, B=@n18)) {materialFunction=&ScaleUVsByCenter}.UVs) {materialFunction=&BreakOutFloat2Components}
  n20 = Add(A=@n19.R, B=DynamicParameter() {paramNames [BottomFade,HeadFade,Erode,UOffset]}.UOffset)
  n21 = ScalarParameter() {parameterName=PanSpeed}
  n22 = Constant()
  n23 = TextureSample(UVs=Panner(Coordinate=AppendVector(A=@n20, B=@n19.G), Speed=AppendVector(A=@n21, B=@n22))) {texture=/Game/Effects/Fort_Effects/Textures/Fire/T_Stylized_Fire_Trail}
  n24 = ParticleColor()
  n25 = MaterialFunctionCall(UV Channel=TextureCoordinate() {coordinateIndex=1}) {materialFunction=&LinearGradient}
  n26 = DynamicParameter() {paramNames [BottomFade,HeadFade,Erode,UOffset]}
  n27 = Multiply(A=OneMinus(None=@n25.UGradient), B=LinearInterpolate(B=OneMinus(None=@n25.VGradient), Alpha=@n26.BottomFade) {constA=1})
  n28 = MaterialFunctionCall() {materialFunction=&LinearGradient}
  n29 = Multiply(A=@n27, B=Saturate(None=Multiply(A=@n28.UGradient)))
  n30 = Multiply(A=@n23.R, B=Multiply(A=@n24.A, B=LinearInterpolate(A=@n29, B=@n27, Alpha=@n26.HeadFade)))
  n31 = LinearInterpolate(Alpha=DynamicParameter() {paramNames [BottomFade,HeadFade,Erode,UOffset]}.Erode) {constA=1, constB=0}
  DepthFade(Opacity=Multiply(A=Subtract(A=@n30, B=@n31), B=ScalarParameter() {defaultValue=2, parameterName=Opacity})) {fadeDistanceDefault=20}
  MaterialFunctionCall() {materialFunction=&DebugTimeSine}
  n32 = CurveAtlasRowParameter(CurveTime=@n23.R) {
   atlas=/Game/Effects/Fort_Effects/GameplayPulginsAssets/Demon/CA_GoldenFire
   curve=/Game/Effects/Fort_Effects/GameplayPulginsAssets/Demon/Curve_GoldenFire01
   parameterName=Param_2
  }
  n33 = Multiply(A=@n32)
  n34 = ScalarParameter() {defaultValue=3, parameterName=Emissive}
  Multiply(
   A=@n24.RGB
   B=MaterialFunctionCall(Emissive=Multiply(A=@n33, B=@n34)) {materialFunction=&MF_FX_EmissiveAdjustments}.Result
  ) {constA=1}
  NamedRerouteDeclaration(None=@n15) {name=UVDistortionOpposite, nodeColor=<rgba 0 0.7656 1 1>}
 ]
}
